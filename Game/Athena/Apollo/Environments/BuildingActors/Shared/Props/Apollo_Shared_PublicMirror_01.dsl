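; Apollo public mirror prop that detects players standing near it or facing it.
; TagAddVolume overlaps apply gameplay effects that add or remove a near-mirror tag on player pawns.
; IsPlayerLookingVolume starts a looping timer that compares pawn yaw to the mesh direction, broadcasting CustomVerb.LookedAtMirror when nearly aligned.
; It stores only one actor and timer handle, and any end overlap on that volume clears the stored timer.
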(event OnComponentEndOverlap(TagAddVolume)
  (params (OverlappedComponent :type Primitive%20Component%20Object%20Reference) (OtherActor :type Actor%20Object%20Reference) (OtherComp :type Primitive%20Component%20Object%20Reference) (OtherBodyIndex :type Integer))
  (Utilities|IsValid OtherActor
    (:Is%20Valid
      (bind _asfort_player_pawn (Utilities|Casting|CastToFortPlayerPawn OtherActor))
      (Creative|GameplayEffects|SimpleApplyGameplayEffectToActor "/Game/Athena/Apollo/Environments/BuildingActors/Shared/Props/GE_PlayerNearMirror_TagRemove.GE_PlayerNearMirror_TagRemove_C" _asfort_player_pawn))))

(event OnComponentBeginOverlap(TagAddVolume)
  (params (OverlappedComponent :type Primitive%20Component%20Object%20Reference) (OtherActor :type Actor%20Object%20Reference) (OtherComp :type Primitive%20Component%20Object%20Reference) (OtherBodyIndex :type Integer) (bFromSweep :type Boolean) (SweepResult :type Hit%20Result%20Structure%20%28by%20ref%29))
  (Utilities|IsValid OtherActor
    (:Is%20Valid
      (bind _asfort_player_pawn (Utilities|Casting|CastToFortPlayerPawn OtherActor))
      (Creative|GameplayEffects|SimpleApplyGameplayEffectToActor "/Game/Athena/Apollo/Environments/BuildingActors/Shared/Props/GE_PlayerNearMirror_TagAdd.GE_PlayerNearMirror_TagAdd_C" _asfort_player_pawn))))

(event OnComponentBeginOverlap(IsPlayerLookingVolume)
  (params (OverlappedComponent :type Primitive%20Component%20Object%20Reference) (OtherActor :type Actor%20Object%20Reference) (OtherComp :type Primitive%20Component%20Object%20Reference) (OtherBodyIndex :type Integer) (bFromSweep :type Boolean) (SweepResult :type Hit%20Result%20Structure%20%28by%20ref%29))
  (Utilities|IsValid OtherActor
    (:Is%20Valid
      (Utilities|Casting|CastToFortPlayerPawn OtherActor)
      (Variables|Default|SetOther%20Actor OtherActor)
      (bind _returnvalue (Utilities|Time|SetTimerbyEvent (delegate Custom|Tick) 1.0 true true 0.5))
      (Variables|Default|SetTimer%20Handle _returnvalue))))

(event Custom|Tick
  (Utilities|IsValid (Variables|Default|GetOther%20Actor)
    (:Is%20Valid
      (if (Math|Float|NearlyEqual(Float) (.yaw (Transformation|GetActorRotation (Game|Player|GetPlayerPawn (Fortnite|GetFortPlayerControllerfromActor (Variables|Default|GetOther%20Actor))))) (.yaw (Math|Rotator|FindLookatRotation (Transformation|GetActorLocation (Game|Player|GetPlayerPawn (Fortnite|GetFortPlayerControllerfromActor (Variables|Default|GetOther%20Actor)))) (Transformation|GetWorldLocation (Variables|Debug|GetStaticMeshComponent)))) 15.0)
        (CustomVerb|BroadcastCustomVerbMessage (CustomVerb|MakeCustomVerbMessage (Variables|Default|GetOther%20Actor) "(TagName=\"CustomVerb.LookedAtMirror\")"))))))

(event OnComponentEndOverlap(IsPlayerLookingVolume)
  (params (OverlappedComponent :type Primitive%20Component%20Object%20Reference) (OtherActor :type Actor%20Object%20Reference) (OtherComp :type Primitive%20Component%20Object%20Reference) (OtherBodyIndex :type Integer))
  (bind _timer_handle (Variables|Default|GetTimer%20Handle))
  (if (Utilities|Time|IsValidTimerHandle _timer_handle)
    (Utilities|Time|ClearandInvalidateTimerbyHandle _timer_handle)))

(fn ConstructionScript)
